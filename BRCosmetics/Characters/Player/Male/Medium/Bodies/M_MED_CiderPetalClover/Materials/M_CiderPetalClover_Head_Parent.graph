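# Parent head material for the CiderPetalClover character cosmetic, adding a timed gold reactive effect over the facial-expression base.
# A Timestamp and duration drive a rise-and-fall curve, overridable by parameter, which a Tech Art Mask channel limits into the reactivity mask.
# The gold layer combines a gradient-texture thin film, optionally tangent-based, with Fresnel emissive dimmed at night.
# Material attribute blending layers it between the 2023 character PreFX and PostFX functions.
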Material /BRCosmetics/Characters/Player/Male/Medium/Bodies/M_MED_CiderPetalClover/Materials/M_CiderPetalClover_Head_Parent
graph {
 properties {
  bUseMaterialAttributes=true
  bUsedWithClothing=true
  bUsedWithMeshDeformer=true
  bUsedWithMorphTargets=true
  bUsedWithNanite=true
  bUsedWithNiagaraMeshParticles=true
  bUsedWithSkeletalMesh=true
  pixelDepthOffsetMode=PDOM_Legacy
  previewMesh=/CosmeticShoes/Assets/FORT_Shoes/Shoe_DareBergAdmit/Meshes/DareBergAdmit_Shoe
  shadingModel=MSM_FromMaterialExpression
  usedShadingModels=Default Lit | Subsurface | Clear Coat | Subsurface Profile | Cloth
 }
 nodes [
  n1 = MaterialFunctionCall() {materialFunction=&MF_CharacterParent_2023_PreFX}
  MaterialFunctionCall(Material=NamedRerouteUsage(), M=@n1.M, Pre Skinned Local Position=@n1.M, Distance Blend=@n1.M, Fresnel=@n1.M) {materialFunction=&MF_CharacterParent_2023_PostFX}
  NamedRerouteDeclaration(None=MaterialFunctionCall!(Material=@n1.Result) {materialFunction=/Game/Global/Materials/Character_Parent_2023/FacialExpression_Parent_2023/MF_FacialExpressionAdvanced_2023}.Result) {name=Mat_In, nodeColor=<rgba 0 0.4609 1 1>}
  n2 = Divide(
   A=Subtract(A=Time(), B=ScalarParameter() {defaultValue=-10000, group=CiderPetalClover - Reactivity, parameterName=Timestamp})
   B=ScalarParameter() {defaultValue=7, group=CiderPetalClover - Reactivity, parameterName=Reactivity_Duration}
  )
  n3 = Saturate(None=LinearInterpolate(A=@n2, B=Frac(None=@n2), Alpha=ScalarParameter() {group=CiderPetalClover - Reactivity, parameterName=Reactivity_LoopPreview}))
  n4 = Abs(None=ConstantBiasScale(None=@n3) {bias=-0.5, scale=2})
  n5 = LinearInterpolate(
   A=ScalarParameter() {defaultValue=2, group=CiderPetalClover - Reactivity, parameterName=Reactivity_StartCurveExp}
   B=ScalarParameter() {defaultValue=0.5, group=CiderPetalClover - Reactivity, parameterName=Reactivity_EndCurveExp}
   Alpha=Round(None=@n3)
  ) {constA=2, constB=0.5}
  n6 = LinearInterpolate(
   A=OneMinus(None=Power(Base=@n4, Exp=@n5))
   B=ScalarParameter() {parameterName=ReactivityCurve}
   Alpha=ScalarParameter() {parameterName=UseReactivityCurve}
  )
  NamedRerouteDeclaration(None=Multiply(
    A=ChannelMaskParameter(None=@n1.Tech Art Mask) {group=CiderPetalClover - Reactivity, parameterName=ReactivityMaskChannel, sortPriority=0}.R
    B=@n6
   )) {name=ReactivityMask, nodeColor=<rgba 0 0.3906 1 1>}
  n7 = Constant3Vector() {constant=<rgba 1 0.3333 0 0>}
  n8 = VectorParameter() {group=CiderPetalClover - Gold Values, parameterName=ReactivityGold_BC, defaultValue=<rgba 1 0.8333 0 1>}
  n9 = Saturate(None=Multiply(A=@n6) {constB=2})
  n10 = Fresnel() {exponent=1}
  n11 = VertexTangentWS()
  n12 = VertexNormalWS()
  n13 = TextureSample() {
   samplerType=SAMPLERTYPE_Masks
   texture=/BRCosmetics/Characters/Player/Male/Medium/Bodies/M_MED_CiderPetalClover/Textures/T_CiderPetalClover_Body_FX2
  }
  n14 = DotProduct(A=LinearInterpolate(A=@n11, B=CrossProduct(A=@n11, B=@n12), Alpha=@n13.G), B=CameraVectorWS())
  n15 = StaticSwitchParameter(True=LinearInterpolate(A=@n10, B=Abs(None=@n14), Alpha=@n13.R), False=@n10) {group=CiderPetalClover - Gold Values, parameterName=UseTangentThinFIlm}
  n16 = ScalarParameter() {defaultValue=1, group=CiderPetalClover - Gold Values, parameterName=ReactivityGold_ThinFilmWarp}
  n17 = TextureSampleParameter2D(UVs=Multiply(
    A=Power(Base=@n15, Exp=@n16)
    B=ScalarParameter() {defaultValue=2, group=CiderPetalClover - Gold Values, parameterName=ReactivityGold_ThinFilmScale}
   )) {
   group=CiderPetalClover - Gold Values
   parameterName=ReactivityGold_ThinFilmTexture
   samplerSource=SSM_Wrap_WorldGroupSettings
   sortPriority=1
   texture=/Game/Global/Materials/Cosmetics/Ranked/Textures/T_Ranked_Gradient_Gold
  }
  n18 = VectorParameter() {group=CiderPetalClover - Gold Values, parameterName=ReactivityGold_Emissive, defaultValue=<rgba 20 3.3333 0 1>}
  n19 = Fresnel(ExponentIn=ScalarParameter() {defaultValue=0.7, group=CiderPetalClover - Gold Values, parameterName=ReactivityGold_FresnelExp})
  n20 = OneMinus(None=@n19)
  n21 = ScalarParameter() {group=CiderPetalClover - Gold Values, parameterName=ReactivityGold_InvertFresnel}
  n22 = Multiply(
   A=@n18.RGB
   B=MaterialFunctionCall(
    In=LinearInterpolate(A=@n19, B=@n20, Alpha=@n21)
    Contrast=ScalarParameter() {defaultValue=0.25, group=CiderPetalClover - Gold Values, parameterName=ReactivityGold_FresnelContrast}
   ) {materialFunction=&CheapContrast}.Result
  )
  n23 = NamedRerouteUsage()
  n24 = Constant() {r=1}
  n25 = Constant() {r=0.5}
  n26 = SetMaterialAttributes!(
   MaterialAttributes=NamedRerouteUsage()
   Base Color=Multiply(A=LinearInterpolate(A=@n7, B=@n8.RGB, Alpha=@n9), B=@n17.RGB)
   Metallic=ScalarParameter() {defaultValue=1, group=CiderPetalClover - Gold Values, parameterName=ReactivityGold_Metallic}
   Specular=ScalarParameter() {defaultValue=0.8, group=CiderPetalClover - Gold Values, parameterName=ReactivityGold_Specular}
   Roughness=ScalarParameter() {defaultValue=0.3, group=CiderPetalClover - Gold Values, parameterName=ReactivityGold_Roughness}
   Emissive Color=Multiply(
    A=Multiply(A=@n22, B=@n23)
    B=MaterialFunctionCall(Value_Morning=@n24, Value_Day=@n24, Value_Evening=@n24, Value_Night=@n25) {materialFunction=&Scalar_Time_of_Day}.Result
   )
  ) {attributeSetTypes [
    #1
    #2
    #3
    #4
    #5
   ]}
  NamedRerouteDeclaration(None=BlendMaterialAttributes(A=NamedRerouteUsage(), B=@n26, Alpha=NamedRerouteUsage())) {name=Mat_Out, nodeColor=<rgba 0.9688 0 1 1>}
 ]
}
